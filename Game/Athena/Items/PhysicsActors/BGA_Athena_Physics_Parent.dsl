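; Base Blueprint for Athena physics props that wake when their support is destroyed and splash when entering water.
; While asleep, the physics object runs a downward link check through FortLinkToActor.
; Destruction of the linked actor sets Rep_WakeUp, whose RepNotify wakes the object; waking clears the flag.
; Buoyancy water entry forces a net update and fires a replicated gameplay cue carrying water surface data and object weight.
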(event EventBeginPlay
  (bind _fortphysicsobject (Variables|Default|GetFortPhysicsObject))
  (bind _buoyancycomponent (Variables|/Script/FortniteGame.FortPhysicsObjectComponent|GetBuoyancyComponent _fortphysicsobject))
  ;[1 1 1] Begin Play - If not awake on begin play then perform a link check
  (if (not (PhysicsObject|IsAwake (Variables|Default|GetFortPhysicsObject)))
    (LinkToActorComponent|PerformLinkingCheck (Variables|Default|GetFortLinkToActor) "Down" (Variables|Default|GetLinkedActor)))
  (Utilities|IsValid _buoyancycomponent
    (:Is%20Valid
      (Cosmetic|BindEventtoOnEnteredWaterDelegate _buoyancycomponent (delegate Custom|OnEnterWater)))))

(event Custom|OnEnterWater
  (params (Pontoon :type Spherical%20Pontoon%20Structure%20%28by%20ref%29))
  ;[1 1 1] Bind To the enter Water Event
  (CallFunction|PlaySplashGC))

(event OnAwakeChanged(FortPhysicsObject)
  (params (SimulatingComponent :type Primitive%20Component%20Object%20Reference) (bIsAwake :type Boolean))
  ;[1 1 1] Handle Binding and waking up when the piece under the actor is broken.
  (if (not bIsAwake)
    (LinkToActorComponent|PerformLinkingCheck (Variables|Default|GetFortLinkToActor) "Down" (Variables|Default|GetLinkedActor))
    (Variables|Default|SetRep_WakeUp false)))

(event Custom|PlaySplashGC
  (bind _fortphysicsobject (Variables|Default|GetFortPhysicsObject))
  (bind _buoyancycomponent (Variables|/Script/FortniteGame.FortPhysicsObjectComponent|GetBuoyancyComponent _fortphysicsobject))
  ;[1 1 1] PlaySplash FX
  (Utilities|IsValid _buoyancycomponent
    (:Is%20Valid
      (bind (_outwaterplanelocation _outwaterplanenormal _outwatersurfaceposition _outwaterdepth _outwaterbodyidx) (Buoyancy|GetLastWaterSurfaceInfo _buoyancycomponent))
      (Networking|ForceNetUpdate)
      (Ability|ExecuteGameplayCueWithParamsReplicated (Variables|Default|GetGC_EnterWater) (Ability|GameplayCue|MakeGameplayCueParameters :NormalizedMagnitude (Math|Conversions|ToFloat(Integer) _outwaterbodyidx) :RawMagnitude _outwaterdepth :Location _outwaterplanelocation :Normal "0,0,0" :GameplayEffectLevel 1 :AbilityLevel (Math|Conversions|ToInteger(Byte) (Variables|Default|GetObjectWeights)) :bReplicateLocationWhenUsingMinimalRepProxy false)))))

(event OnLinkedActorDestroyed(FortLinkToActor)
  (params (DamageCauser :type Actor%20Object%20Reference))
  ;[1 1 1] Handle Binding and waking up when the piece under the actor is broken.
  (Variables|Default|SetRep_WakeUp true))

(fn OnRep_Rep_WakeUp
  ;[1 1 1] Rep Waking up.
  (if (Variables|Default|GetRep_WakeUp)
    (PhysicsObject|WakeUp (Variables|Default|GetFortPhysicsObject) "Normal" :!class /Script/FortniteGame.FortPhysicsObjectComponent)))

(fn ConstructionScript)
