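; Parent Blueprint for physics-based building containers that wake up and simulate when the piece beneath them is destroyed.
; Whenever the physics object is asleep at BeginPlay or falls asleep later, the LinkToActor component runs a downward linking check.
; Destruction of the linked actor sets the replicated Rep_WakeUp flag, whose OnRep wakes the FortPhysicsObject component.
; The flag clears once the object reports being awake.
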(event EventBeginPlay
  ;[1 1 1] Begin Play - If not awake on begin play then perform a link check
  (if (not (PhysicsObject|IsAwake (Variables|Default|GetFortPhysicsObject)))
    (LinkToActorComponent|PerformLinkingCheck (Variables|Default|GetFortLinkToActor) "Down" (Variables|Default|GetLinkedActor))))

(event OnAwakeChanged(FortPhysicsObject)
  (params (SimulatingComponent :type Primitive%20Component%20Object%20Reference) (bIsAwake :type Boolean))
  ;[1 1 1] Handle Binding and waking up when the piece under the actor is broken.
  (if (not bIsAwake)
    (LinkToActorComponent|PerformLinkingCheck (Variables|Default|GetFortLinkToActor) "Down" (Variables|Default|GetLinkedActor))
    (Variables|Default|SetRep_WakeUp false)))

(event OnLinkedActorDestroyed(FortLinkToActor)
  (params (DamageCauser :type Actor%20Object%20Reference))
  ;[1 1 1] Handle Binding and waking up when the piece under the actor is broken.
  (Variables|Default|SetRep_WakeUp true))

(fn OnRep_Rep_WakeUp
  ;[1 1 1] Rep Waking up.
  (if (Variables|Default|GetRep_WakeUp)
    (PhysicsObject|WakeUp (Variables|Default|GetFortPhysicsObject) "Normal" :!class /Script/FortniteGame.FortPhysicsObjectComponent)))

(fn ConstructionScript)
